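; Looping GreaseClipDunmurry in-vehicle gameplay cue that tracks which local player controllers are viewing the target pawn.
; Outside dedicated servers, locally-viewed changes start a 0.25-second timer that diffs viewers, stopping on removal or once none remain.
; Each gained or lost viewer adds or removes a local GCNL_LocalViewer cue, passing the controller as SourceObject.
; View checks resolve spectators' followed players and GreaseClip cine cameras bound to sequences the pawn instigated.
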(event EventOnLoopingStart(EitherVFX-preferthisversion)
  (params (MyTarget :type Actor%20Object%20Reference) (Parameters :type Gameplay%20Cue%20Parameters%20Structure%20%28by%20ref%29) (ParticleComponents :type Array%20of%20FXSystem%20Component%20Object%20References) (AudioComponents :type Array%20of%20Audio%20Component%20Object%20References))
  (bind _targetplayerpawn (Variables|Default|GetTargetPlayerPawn))
  ;[1 1 1] Start / End Effect
  (Variables|Default|SetCueInstigator (Ability|GameplayCue|GetInstigatorActor Parameters))
  (bind _asfort_player_pawn (Utilities|Casting|CastToFortPlayerPawn MyTarget))
  (bind _output_get (Variables|Default|SetTargetPlayerPawn _asfort_player_pawn))
  (if (not (Networking|IsDedicatedServer _output_get))
    (Variables|Default|SetIsGCNLActive true)
    (EventDispatchers|BindEventtoOnIsLocallyViewedChanged _targetplayerpawn (EventDispatchers|CreateEvent :!event Custom|HandleLocallyViewedChanged))
    (CallFunction|HandleLocallyViewedChanged (Pawn|IsLocallyViewed _targetplayerpawn))))

(event EventOnRemoval(EitherVFX-preferthisversion)
  (params (MyTarget :type Actor%20Object%20Reference) (Parameters :type Gameplay%20Cue%20Parameters%20Structure%20%28by%20ref%29) (ParticleComponents :type Array%20of%20FXSystem%20Component%20Object%20References) (AudioComponents :type Array%20of%20Audio%20Component%20Object%20References) (BurstCameraShakeInstance :type Camera%20Shake%20Base%20Object%20Reference) (BurstDecalInstance :type Decal%20Actor%20Object%20Reference))
  ;[1 1 1] Start / End Effect
  (Variables|Default|SetIsGCNLActive false)
  (bind _targetplayerpawn (Variables|Default|GetTargetPlayerPawn)
    (:then
      (EventDispatchers|UnbindEventfromOnIsLocallyViewedChanged _targetplayerpawn (EventDispatchers|CreateEvent :!event Custom|HandleLocallyViewedChanged)))
    (:else
      _))
  (Variables|Default|GetIsLocalViewerTickRunning
    (:then
      (CallFunction|EndLocalViewerTick))
    (:else
      _))
  (Variables|Default|SetTargetPlayerPawn 0)
  (Variables|Default|SetCueInstigator 0))

(event Custom|HandleLocallyViewedChanged
  (params (IsLocallyViewed :type Boolean))
  ;[1 1 1] Local Viewer Tick Init & Teardown
  (if (and IsLocallyViewed (Variables|Default|GetIsGCNLActive))
    (Variables|Default|GetIsLocalViewerTickRunning
      (:then)
      (:else
        (CallFunction|StartLocalViewerTick)))
    (else
      (Variables|Default|GetIsLocalViewerTickRunning
        (:then
          (CallFunction|RecalculateLocalViewers)
          (if (Utilities|Array|IsEmpty (Variables|Default|GetLocalViewers))
            (CallFunction|EndLocalViewerTick)))
        (:else)))))

(event Custom|StartLocalViewerTick
  ;[1 1 1] Local Viewer Tick Init & Teardown
  (Variables|Default|SetIsLocalViewerTickRunning true)
  (bind _returnvalue (Utilities|Time|SetTimerbyEvent (EventDispatchers|CreateEvent :!event RecalculateLocalViewers) 0.25 true true))
  (Variables|Default|SetUpdateTimerHandle _returnvalue)
  (CallFunction|RecalculateLocalViewers))

(event Custom|EndLocalViewerTick
  (bind _localviewers (Variables|Default|GetLocalViewers))
  ;[1 1 1] Local Viewer Tick Init & Teardown
  (Variables|Default|SetIsLocalViewerTickRunning false)
  (Utilities|Time|ClearandInvalidateTimerbyHandle self (Variables|Default|GetUpdateTimerHandle))
  (for _array_element _localviewers
    (CallFunction|LostLocalViewer _array_element))
  (Utilities|Array|Clear _localviewers))

(event Custom|GainedLocalViewer
  (params (AddedLocalViewer :type Fort%20Player%20Controller%20Object%20Reference))
  ;[1 1 1] Gained/Lost Local Viewer
  (Projectile|AddGameplayCue%28Looping%29Local (Variables|Default|GetTargetPlayerPawn) (Variables|Default|GetGCNL_LocalViewer) (Ability|GameplayCue|MakeGameplayCueParameters :NormalizedMagnitude 0.0 :RawMagnitude 0.0 :Location "0,0,0" :Normal "0,0,0" :Instigator (Variables|Default|GetCueInstigator) :SourceObject AddedLocalViewer :GameplayEffectLevel 1 :AbilityLevel 1 :bReplicateLocationWhenUsingMinimalRepProxy false) false))

(event Custom|LostLocalViewer
  (params (RemovedLocalViewer :type Fort%20Player%20Controller%20Object%20Reference))
  ;[1 1 1] Gained/Lost Local Viewer
  (Projectile|RemoveGameplayCue%28Looping%29Local (Variables|Default|GetTargetPlayerPawn) (Variables|Default|GetGCNL_LocalViewer) (Ability|GameplayCue|MakeGameplayCueParameters :NormalizedMagnitude 0.0 :RawMagnitude 0.0 :Location "0,0,0" :Normal "0,0,0" :Instigator (Variables|Default|GetCueInstigator) :SourceObject RemovedLocalViewer :GameplayEffectLevel 1 :AbilityLevel 1 :bReplicateLocationWhenUsingMinimalRepProxy false) false))

(fn IsLocalViewer
  (params (PlayerController :type Fort%20Player%20Controller%20Object%20Reference))
  (Utilities|IsValid PlayerController
    (:Is%20Valid
      ;[1 1 1 0.5] Calculate effective view target
      (bind _asfort_player_controller_spectating (Utilities|Casting|CastToFortPlayerControllerSpectating PlayerController)
        (:then
          ;; Broadcast spectators may have their view target set to the vehicle, but we want to know about the pawn that caused them to view the vehicle, which is who they are "following"
          (Utilities|IsValid (Class|FortPlayerControllerSpectating|GetFollowedPlayer _asfort_player_controller_spectating)
            (:Is%20Valid
              (LocalVariables|SetViewTarget (PlayerState|GetCurrentPawn (Class|FortPlayerControllerSpectating|GetFollowedPlayer _asfort_player_controller_spectating))))
            (:Is%20Not%20Valid
              (LocalVariables|SetViewTarget 0))))
        (:CastFailed
          (LocalVariables|SetViewTarget (Pawn|GetViewTarget PlayerController))))
      ;[1 1 1] Are we viewing the pawn?
      (if (== (LocalVariables|GetViewTarget) (Variables|Default|GetTargetPlayerPawn))
        (return true)
        (else
          ;[1 1 1] Are we viewing SOME GreaseClip cine-cam?
          (bind _asgrease_clip_cine_camera (Utilities|Casting|CastToGreaseClipCineCamera (LocalVariables|GetViewTarget))
            (:then
              (LocalVariables|SetViewTargetAsCineCam _asgrease_clip_cine_camera)
              ;[1 1 1] Is the GreaseClip cine-cam we are viewing associated with the target pawn?
              (bind _outactors (Actor|GetAllActorsOfClass (LocalVariables|GetViewTarget) "/GreaseClipDunmurry/ProtoSeq/GreaseClipLevelSequenceActor.GreaseClipLevelSequenceActor_C"))
              (for _array_element _outactors
                ;;[1 1 1] Is this Sequence associated with the target pawn?
                (if (== (Game|GetInstigator _array_element) (Variables|Default|GetTargetPlayerPawn))
                  ;;[1 1 1] Is this sequence's camera the one we are viewing?
                  (bind _sequenceplayer (Variables|/Script/LevelSequence.LevelSequenceActor|GetSequencePlayer _array_element)
                    (:then
                      (bind _returnvalue (Sequencer|Player|GetObjectBindings _sequenceplayer (LocalVariables|GetViewTargetAsCineCam) :!class /Script/MovieScene.MovieSceneSequencePlayer))
                      (if (not (Utilities|Array|IsEmpty _returnvalue))
                        (return true)))
                    (:else))))
              (return false))
            (:CastFailed
              (return false))))))
    (:Is%20Not%20Valid
      (return true))))

(fn RecalculateLocalViewers
  (bind _targetplayerpawn (Variables|Default|GetTargetPlayerPawn)
    (:then
      ;[1 1 1] Calculate current viewers
      (for _array_element (Fortnite|GetLocalFortPlayerControllers _targetplayerpawn true true)
        (bind _islocalviewer (CallFunction|IsLocalViewer _array_element))
        (if _islocalviewer
          (Utilities|Array|Add (LocalVariables|GetNewLocalViewers) _array_element)))
      ;[1 1 1] Process lost viewers
      (for _array_element_2 (Variables|Default|GetLocalViewers)
        (if (not (Utilities|Array|ContainsItem (LocalVariables|GetNewLocalViewers) _array_element_2))
          (CallFunction|LostLocalViewer _array_element_2)))
      ;[1 1 1] Process acquired viewers
      (for _array_element_3 (LocalVariables|GetNewLocalViewers)
        (if (not (Utilities|Array|ContainsItem (Variables|Default|GetLocalViewers) _array_element_3))
          (CallFunction|GainedLocalViewer _array_element_3)))
      (Variables|Default|SetLocalViewers (LocalVariables|GetNewLocalViewers)))
    (:else)))

(fn ConstructionScript)
