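; Looping GameplayCue notify that runs the local viewer's presentation of the GreaseClipDunmurry vehicle mod's boost-driven time tunnel.
; Start caches pawn, controller and mod, broadcasts a started event and builds a camera modifier.
; Tick smooths intro progress to drive camera, music volume, HUD hiding and a speed-gated white fade, spawning a per-local-player offset level sequence once the vehicle hides locally.
; Removal undoes look-input, spectator-tag and HUD changes, then recycles.
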(event EventOnLoopingStart(EitherVFX-preferthisversion)
  (params (MyTarget :type Actor%20Object%20Reference) (Parameters :type Gameplay%20Cue%20Parameters%20Structure%20%28by%20ref%29) (ParticleComponents :type Array%20of%20FXSystem%20Component%20Object%20References) (AudioComponents :type Array%20of%20Audio%20Component%20Object%20References))
  (bind (_normalizedmagnitude _rawmagnitude _effectcontext _matchedtagname _originaltag _aggregatedsourcetags _aggregatedtargettags _location _normal _instigator _effectcauser _sourceobject) (Ability|GameplayCue|BreakGameplayCueParameters Parameters))
  (bind _observingplayercontroller (Variables|Default|GetObservingPlayerController))
  ;[1 1 1] Start
  (bind _asfort_player_pawn (Utilities|Casting|CastToFortPlayerPawn MyTarget))
  (Variables|Default|SetInVehiclePawn _asfort_player_pawn)
  (bind _asplayer_controller (Utilities|Casting|CastToPlayerController _sourceobject))
  (Variables|Default|SetObservingPlayerController _asplayer_controller)
  (bind _asvm_bga_grease_clip_dunmurry (Utilities|Casting|CastToVM_BGA_GreaseClipDunmurry _instigator))
  (Variables|Default|SetVehicleMod _asvm_bga_grease_clip_dunmurry)
  (CallFunction|BuildCameraModifier)
  ;[1 1 1] Broadcast Started Event
  (bind _returnvalue (GameplayEventRouter|GetEventRouter "/Script/FortniteGame.FortPlayspace" _observingplayercontroller))
  (GameplayEventRouter|BroadcastStatefulEvent _returnvalue (Utilities|Struct|MakeOnGreaseClipDunmurryStarted true) _observingplayercontroller))

(event EventOnRemoval(EitherVFX-preferthisversion)
  (params (MyTarget :type Actor%20Object%20Reference) (Parameters :type Gameplay%20Cue%20Parameters%20Structure%20%28by%20ref%29) (ParticleComponents :type Array%20of%20FXSystem%20Component%20Object%20References) (AudioComponents :type Array%20of%20Audio%20Component%20Object%20References) (BurstCameraShakeInstance :type Camera%20Shake%20Base%20Object%20Reference) (BurstDecalInstance :type Decal%20Actor%20Object%20Reference))
  ;[1 1 1] End
  (Utilities|FlowControl|Sequence
    (:then_0
      (bind _lenseffectactor (Variables|Default|GetLensEffectActor)
        (:then
          (Actor|DestroyActor _lenseffectactor)
          (Variables|Default|SetLensEffectActor 0))
        (:else)))
    (:then_1
      ;;[1 1 1] Clear Started Event
      (bind _returnvalue (GameplayEventRouter|GetEventRouter "/Script/FortniteGame.FortPlayspace" (Variables|Default|GetObservingPlayerController)))
      (GameplayEventRouter|ClearEventState _returnvalue "/Script/GreaseClipDunmurryRuntime.OnGreaseClipDunmurryStarted"))
    (:then_2
      (CallFunction|ResetDoOnceAudio)
      (CallFunction|ResetDoOnceCamera)
      (CallFunction|ResetTimeTunnel)
      (bind _introsequencecameramodifier (Variables|Default|GetIntroSequenceCameraModifier)
        (:then
          (CameraModifier|DisableModifier _introsequencecameramodifier true))
        (:else)))
    (:then_3
      (CallFunction|DestroyCameraModifier)
      (Variables|Default|GetInTimeTunnelSequence
        (:then
          (if (ScalableFloat|GetValueasBool (Variables|Default|GetHF_DoLockLookInputDuringSequence))
            (Input|SetIgnoreLookInput (Variables|Default|GetObservingPlayerController) false)))
        (:else)))
    (:then_4
      (CallFunction|RelockSpecatorViews))
    (:then_5
      (Variables|Default|GetIsHUDHidden
        (:then
          (bind _observingplayercontroller (Variables|Default|GetObservingPlayerController)
            (:then
              (Utilities|FlowControl|Delay _observingplayercontroller (ScalableFloat|GetValueatLevel (Variables|Default|GetHF_RestoreHUDAfterSequenceDelaySeconds)))
              (bind _asfort_player_controller (Utilities|Casting|CastToFortPlayerController (Variables|Default|GetObservingPlayerController))
                (:then
                  (HUD|SetHUDElementVisibilitywithContext _asfort_player_controller (Variables|Default|GetHUDElement%20Tags) false "None" "")
                  (bind _returnvalue_2 (GameplayEventRouter|GetEventRouter "/Script/FortniteGame.FortPlayspace" _asfort_player_controller))
                  (GameplayEventRouter|BroadcastStatefulEvent _returnvalue_2 (Utilities|Struct|MakeOnGreaseClipDunmurryShouldHideUI false) _asfort_player_controller)
                  (CallFunction|FinishCleanupandRecycle))
                (:CastFailed
                  (CallFunction|FinishCleanupandRecycle))))
            (:else
              _)))
        (:else
          _))
      (CallFunction|FinishCleanupandRecycle))))

(event Custom|UnlockSpectatorViews
  ;[1 1 1 0.5] Unlock/Relock Spectator Views
  (Utilities|FlowControl|Sequence
    (:then_0
      (bind _invehiclepawn (Variables|Default|GetInVehiclePawn)
        (:then
          (bind _returnvalue (Ability|GameplayEffect|AddLooseGameplayTags _invehiclepawn (Variables|Default|GetUnlockSpectatorViewTag)))
          (Variables|Default|SetDidAddUnlockSpectatorViewTag _returnvalue))
        (:else)))
    (:then_1
      (bind _vehiclemod (Variables|Default|GetVehicleMod)
        (:then
          (bind _attachedvehicle (Variables|/GreaseClipDunmurry/Gameplay/VM_BGA_GreaseClipDunmurry.VM_BGA_GreaseClipDunmurry_C|GetAttachedVehicle _vehiclemod)
            (:then
              (bind _returnvalue_2 (Ability|GameplayEffect|AddLooseGameplayTags _attachedvehicle (Variables|Default|GetUnlockSpectatorViewTag)))
              (Variables|Default|SetDidAddUnlockSpectatorViewTagVehicle _returnvalue_2))
            (:else)))
        (:else)))))

(event Custom|RelockSpecatorViews
  ;[1 1 1 0.5] Unlock/Relock Spectator Views
  (Utilities|FlowControl|Sequence
    (:then_0
      (Variables|Default|GetDidAddUnlockSpectatorViewTag
        (:then
          (Variables|Default|SetDidAddUnlockSpectatorViewTag false)
          (bind _invehiclepawn (Variables|Default|GetInVehiclePawn)
            (:then
              (Ability|GameplayEffect|RemoveLooseGameplayTags _invehiclepawn (Variables|Default|GetUnlockSpectatorViewTag)))
            (:else)))
        (:else)))
    (:then_1
      (Variables|Default|GetDidAddUnlockSpectatorViewTagVehicle
        (:then
          (Variables|Default|SetDidAddUnlockSpectatorViewTagVehicle false)
          (bind _vehiclemod (Variables|Default|GetVehicleMod)
            (:then
              (bind _attachedvehicle (Variables|/GreaseClipDunmurry/Gameplay/VM_BGA_GreaseClipDunmurry.VM_BGA_GreaseClipDunmurry_C|GetAttachedVehicle _vehiclemod)
                (:then
                  (Ability|GameplayEffect|RemoveLooseGameplayTags _attachedvehicle (Variables|Default|GetUnlockSpectatorViewTag)))
                (:else)))
            (:else)))
        (:else)))))

(event Custom|FinishCleanupAndRecycle
  ;[1 1 1] End
  ;;[1 1 1 0.5] Recycle
  (Variables|Default|SetInVehiclePawn 0)
  (Variables|Default|SetObservingPlayerController 0)
  (Variables|Default|SetVehicleMod 0)
  (Variables|Default|SetStayFullVolume false)
  (Variables|Default|SetInTimeTunnelSequence false)
  (Variables|Default|SetReturnedFromHiddenTimestamp -1.0)
  (Variables|Default|SetIsHUDHidden false)
  (GameplayCueNotify|End%28Recycle%29GameplayCue))

(event Custom|ResetTimeTunnel
  ;[1 1 1] Time Tunnel Sequence
  (Utilities|FlowControl|DoOnce false
    (:Completed
      (Variables|Default|SetStayFullVolume true)
      (Variables|Default|SetInTimeTunnelSequence true)
      (if (ScalableFloat|GetValueasBool (Variables|Default|GetHF_DoLockLookInputDuringSequence))
        (Input|SetIgnoreLookInput (Variables|Default|GetObservingPlayerController) true)
        (else
          _))
      (if (ScalableFloat|GetValueasBool (Variables|Default|GetHF_UnlockSpectatorViewTargetDuringSequence))
        (CallFunction|UnlockSpectatorViews)
        (else
          _))
      (bind _returnvalue (Game|SpawnActorfromClass 0 "/GreaseClipDunmurry/ProtoSeq/GreaseClipLevelSequenceActor.GreaseClipLevelSequenceActor_C" (Math|Transform|MakeTransform (Utilities|Operators|Add (Transformation|GetActorLocation) (Math|Vector|MakeVector 0.0 0.0 (CallFunction|GetSequenceOffsetforPlayer (Variables|Default|GetObservingPlayerController))))) "Undefined" "MultiplyWithRoot" (Variables|Default|GetObservingPlayerController) 0 (Variables|Default|GetPlayback%20Settings) "/GreaseClipDunmurry/ProtoSeq/Seq_GreaseClipProto.Seq_GreaseClipProto" (Variables|Default|GetCamera%20Settings) true (Variables|Default|GetInVehiclePawn)))
      (Utilities|IsValid _returnvalue
        (:Is%20Valid
          (bind _sequenceplayer (Variables|/Script/LevelSequence.LevelSequenceActor|GetSequencePlayer _returnvalue)
            (:then
              (Variables|/GreaseClipDunmurry/ProtoSeq/GreaseClipLevelSequenceActor.GreaseClipLevelSequenceActor_C|SetAssociated%20Vehicle (Variables|/GreaseClipDunmurry/Gameplay/VM_BGA_GreaseClipDunmurry.VM_BGA_GreaseClipDunmurry_C|GetAttachedVehicle (Variables|Default|GetVehicleMod)) _returnvalue)
              (Sequencer|Player|Play _sequenceplayer :!class /Script/MovieScene.MovieSceneSequencePlayer))
            (:else)))))))

(event Custom|ResetDoOnceAudio
  (bind _tunnel_music (Variables|Default|GetTunnel%20Music))
  ;[1 1 1] Activate Audio
  (Utilities|FlowControl|DoOnce false
    (:Completed
      (Audio|Components|Audio|Play _tunnel_music 0.0)
      (Audio|AddEventbyComponent (GameInstanceSubsystems|GetMusicEventSubsystem 0) _tunnel_music "(TagName=\"MusicEventSubsystem.Event.Gameplay.Ability.1P\")" "(GameplayTags=)" 0 ""))))

(event Custom|OnBoostFinishedDelegate_Event
  (bind _tunnel_music (Variables|Default|GetTunnel%20Music))
  ;[1 1 1] Activate Audio
  (bind _vehiclemod (Variables|Default|GetVehicleMod)
    (:then
      (if (not (Class|VMBGAGreaseClipDunmurry|GetIsVehicleHiddenLocal _vehiclemod))
        (Audio|Components|Audio|FadeOut _tunnel_music 0.5 0.0 "Linear" :!class /Script/Engine.AudioComponent)
        (Audio|RemoveEvent (GameInstanceSubsystems|GetMusicEventSubsystem 0) (Variables|Default|GetTunnel%20SFX))
        (Utilities|FlowControl|DoOnce false
          (:Completed
            (Audio|Components|Audio|Play _tunnel_music 0.0)
            (Audio|AddEventbyComponent (GameInstanceSubsystems|GetMusicEventSubsystem 0) _tunnel_music "(TagName=\"MusicEventSubsystem.Event.Gameplay.Ability.1P\")" "(GameplayTags=)" 0 "")))))
    (:else)))

(event EventTick
  (params (DeltaSeconds :type Float%20%28single-precision%29))
  (bind _tunnel_music (Variables|Default|GetTunnel%20Music))
  (CallFunction|UpdateIntroSequenceProgress DeltaSeconds)
  (Utilities|FlowControl|Sequence
    (:then_0
      ;[1 1 1] Time Tunnel Sequence
      (bind _showtimetunnel (CallFunction|ShouldShowTimeTunnel))
      (if _showtimetunnel
        (Utilities|FlowControl|DoOnce false
          (:Completed
            (Variables|Default|SetStayFullVolume true)
            (Variables|Default|SetInTimeTunnelSequence true)
            (if (ScalableFloat|GetValueasBool (Variables|Default|GetHF_DoLockLookInputDuringSequence))
              (Input|SetIgnoreLookInput (Variables|Default|GetObservingPlayerController) true)
              (else
                _))
            (if (ScalableFloat|GetValueasBool (Variables|Default|GetHF_UnlockSpectatorViewTargetDuringSequence))
              (CallFunction|UnlockSpectatorViews)
              (else
                _))
            (bind _returnvalue (Game|SpawnActorfromClass 0 "/GreaseClipDunmurry/ProtoSeq/GreaseClipLevelSequenceActor.GreaseClipLevelSequenceActor_C" (Math|Transform|MakeTransform (Utilities|Operators|Add (Transformation|GetActorLocation) (Math|Vector|MakeVector 0.0 0.0 (CallFunction|GetSequenceOffsetforPlayer (Variables|Default|GetObservingPlayerController))))) "Undefined" "MultiplyWithRoot" (Variables|Default|GetObservingPlayerController) 0 (Variables|Default|GetPlayback%20Settings) "/GreaseClipDunmurry/ProtoSeq/Seq_GreaseClipProto.Seq_GreaseClipProto" (Variables|Default|GetCamera%20Settings) true (Variables|Default|GetInVehiclePawn)))
            (Utilities|IsValid _returnvalue
              (:Is%20Valid
                (bind _sequenceplayer (Variables|/Script/LevelSequence.LevelSequenceActor|GetSequencePlayer _returnvalue)
                  (:then
                    (Variables|/GreaseClipDunmurry/ProtoSeq/GreaseClipLevelSequenceActor.GreaseClipLevelSequenceActor_C|SetAssociated%20Vehicle (Variables|/GreaseClipDunmurry/Gameplay/VM_BGA_GreaseClipDunmurry.VM_BGA_GreaseClipDunmurry_C|GetAttachedVehicle (Variables|Default|GetVehicleMod)) _returnvalue)
                    (Sequencer|Player|Play _sequenceplayer :!class /Script/MovieScene.MovieSceneSequencePlayer))
                  (:else))))))))
    (:then_1
      ;[1 1 1] Activate Audio
      (bind _vehiclemod (Variables|Default|GetVehicleMod)
        (:then
          (Vehicle|Boost|AssignOnBoostFinishedDelegate (Variables|/GreaseClipDunmurry/Gameplay/VM_BGA_GreaseClipDunmurry.VM_BGA_GreaseClipDunmurry_C|GetAttachedVehicle _vehiclemod) (EventDispatchers|CreateEvent :!event Custom|OnBoostFinishedDelegate_Event))
          (if (and (Class|FortVehicleInterface|IsBoosting (Variables|/GreaseClipDunmurry/Gameplay/VM_BGA_GreaseClipDunmurry.VM_BGA_GreaseClipDunmurry_C|GetAttachedVehicle _vehiclemod)) (>= (Vehicle|GetForwardSpeedKmh (Variables|/GreaseClipDunmurry/Gameplay/VM_BGA_GreaseClipDunmurry.VM_BGA_GreaseClipDunmurry_C|GetAttachedVehicle _vehiclemod)) (ScalableFloat|GetValueatLevel (Variables|Default|GetHF_IntroSequenceSpeedRequired))))
            _
            (elif (Class|VMBGAGreaseClipDunmurry|GetIsVehicleHiddenLocal _vehiclemod)
              _
              (elif (Variables|Default|GetIntroSequenceActive)
                _
                (elif (Variables|Default|GetReachedFullVolume)
                  _
                  (else
                    (Utilities|FlowControl|Delay 0 0.5)
                    (Audio|Components|Audio|FadeOut _tunnel_music 0.5 0.0 "Linear" :!class /Script/Engine.AudioComponent)
                    (Audio|RemoveEvent (GameInstanceSubsystems|GetMusicEventSubsystem 0) (Variables|Default|GetTunnel%20SFX)))))))
          (Utilities|FlowControl|DoOnce false
            (:Completed
              (Audio|Components|Audio|Play _tunnel_music 0.0)
              (Audio|AddEventbyComponent (GameInstanceSubsystems|GetMusicEventSubsystem 0) _tunnel_music "(TagName=\"MusicEventSubsystem.Event.Gameplay.Ability.1P\")" "(GameplayTags=)" 0 ""))))
        (:else)))
    (:then_2
      ;[1 1 1] Camera Modifier
      (bind _introsequencecameramodifier (Variables|Default|GetIntroSequenceCameraModifier)
        (:then
          (if (== (CameraModifier|IsDisabled _introsequencecameramodifier) (Variables|Default|GetIntroSequenceActive))
            (Variables|Default|GetIntroSequenceActive
              (:then
                (CameraModifier|EnableModifier _introsequencecameramodifier))
              (:else
                (CameraModifier|DisableModifier _introsequencecameramodifier false)))))
        (:else)))
    (:then_3
      ;[1 1 1] HUD Hiding
      (bind _asfort_player_controller (Utilities|Casting|CastToFortPlayerController (Variables|Default|GetObservingPlayerController)))
      (bind _shouldhide (CallFunction|ShouldHideHUD))
      (if (!= _shouldhide (Variables|Default|GetIsHUDHidden))
        (bind _output_get (Variables|Default|SetIsHUDHidden _shouldhide))
        (HUD|SetHUDElementVisibilitywithContext _asfort_player_controller (Variables|Default|GetHUDElement%20Tags) _output_get "None" "")
        (bind _returnvalue_2 (GameplayEventRouter|GetEventRouter "/Script/FortniteGame.FortPlayspace" (Variables|Default|GetObservingPlayerController)))
        (GameplayEventRouter|BroadcastStatefulEvent _returnvalue_2 (Utilities|Struct|MakeOnGreaseClipDunmurryShouldHideUI _output_get (not (Variables|Default|GetInTimeTunnelSequence))) (Variables|Default|GetObservingPlayerController))))
    (:then_4
      ;[1 1 1] Camera Fade
      (bind _vehiclemod_2 (Variables|Default|GetVehicleMod)
        (:then
          (if (and (Class|FortVehicleInterface|IsBoosting (Variables|/GreaseClipDunmurry/Gameplay/VM_BGA_GreaseClipDunmurry.VM_BGA_GreaseClipDunmurry_C|GetAttachedVehicle _vehiclemod_2)) (>= (Vehicle|GetForwardSpeedKmh (Variables|/GreaseClipDunmurry/Gameplay/VM_BGA_GreaseClipDunmurry.VM_BGA_GreaseClipDunmurry_C|GetAttachedVehicle _vehiclemod_2)) (ScalableFloat|GetValueatLevel (Variables|Default|GetHF_CameraFadeVehicleSpeed))))
            (CallFunction|DoCameraFade)))
        (:else)))
    (:then_5
      ;[1 1 1] Audio Fade In
      (CallFunction|AudioVolumeControl)
      (Audio|Components|Audio|SetVolumeMultiplier (Variables|Default|GetTunnel%20Music) (Variables|Default|GetVolume) :!class /Script/Engine.AudioComponent))))

(event Custom|DoCameraFade
  ;[1 1 1] Camera Fade
  (Utilities|FlowControl|DoOnce false
    (:Completed
      (bind _observingplayercontroller (Variables|Default|GetObservingPlayerController)
        (:then
          (bind _playercameramanager (Variables|/Script/Engine.PlayerController|GetPlayerCameraManager _observingplayercontroller)
            (:then
              (CameraFades|StartCameraFade _playercameramanager 0.0 1.0 0.755 "(R=1.000000,G=1.000000,B=1.000000,A=1.000000)" false false))
            (:else)))
        (:else)))))

(event Custom|ResetDoOnceCamera
  ;[1 1 1] Camera Fade
  (Utilities|FlowControl|DoOnce false
    (:Completed
      (bind _observingplayercontroller (Variables|Default|GetObservingPlayerController)
        (:then
          (bind _playercameramanager (Variables|/Script/Engine.PlayerController|GetPlayerCameraManager _observingplayercontroller)
            (:then
              (CameraFades|StartCameraFade _playercameramanager 0.0 1.0 0.755 "(R=1.000000,G=1.000000,B=1.000000,A=1.000000)" false false))
            (:else)))
        (:else)))))

(fn AudioVolumeControl
  (bind _vehiclemod (Variables|Default|GetVehicleMod)
    (:then
      (if (Variables|Default|GetReachedFullVolume)
        _
        (elif (Class|VMBGAGreaseClipDunmurry|GetIsVehicleHiddenLocal _vehiclemod)
          _
          (elif (Variables|Default|GetStayFullVolume)
            _
            (else
              (Variables|Default|SetVolume (Math|Float|Clamp(Float) (+ (Variables|Default|GetIntroSequenceBlendedPercent) 0.2)))))))
      (Variables|Default|SetVolume 1.0)
      (Variables|Default|SetReachedFullVolume true))
    (:else)))

(fn BuildCameraModifier
  (bind _observingplayercontroller (Variables|Default|GetObservingPlayerController)
    (:then
      (bind _playercameramanager (Variables|/Script/Engine.PlayerController|GetPlayerCameraManager _observingplayercontroller)
        (:then
          (bind _returnvalue (Game|Player|AddNewCameraModifier _playercameramanager "/GreaseClipDunmurry/Gameplay/Camera/CameraModifier_Temp_GreaseClipDunmurry.CameraModifier_Temp_GreaseClipDunmurry_C"))
          (bind _output_get (Variables|Default|SetIntroSequenceCameraModifier _returnvalue))
          (Utilities|IsValid _output_get
            (:Is%20Valid
              (CameraModifier|DisableModifier _output_get false))))
        (:else)))
    (:else)))

(fn DestroyCameraModifier
  (Utilities|FlowControl|Sequence
    (:then_0
      (bind _introsequencecameramodifier (Variables|Default|GetIntroSequenceCameraModifier)
        (:then
          (bind _observingplayercontroller (Variables|Default|GetObservingPlayerController)
            (:then
              (bind _playercameramanager (Variables|/Script/Engine.PlayerController|GetPlayerCameraManager _observingplayercontroller)
                (:then
                  (Game|Player|RemoveCameraModifier _playercameramanager _introsequencecameramodifier))
                (:else)))
            (:else)))
        (:else)))
    (:then_1
      (Variables|Default|SetIntroSequenceCameraModifier 0))))

(fn GetSequenceOffsetForPlayer
  (params (PlayerController :type Player%20Controller%20Object%20Reference))
  (Utilities|FlowControl|Sequence
    (:then_0
      (bind (_array_element _array_index) (Default|ForEachFortLocalPlayerController)
        (:Loop%20Body
          (if (== _array_element PlayerController)
            (LocalVariables|SetLocalPlayerIndex _array_index)))))
    (:then_1
      (return (* (LocalVariables|GetLocalPlayerIndex) 1e+04)))))

(fn ShouldHideHUD
  (bind _vehiclemod (Variables|Default|GetVehicleMod)
    (:then
      (if (Class|VMBGAGreaseClipDunmurry|GetIsVehicleHiddenLocal _vehiclemod)
        (return true)
        (elif (Class|FortVehicleModGreaseClipDunmurry|WasVehicleEverHidden _vehiclemod)
          (bind _output_get (LocalVariables|SetDelaySeconds (ScalableFloat|GetValueatLevel (Variables|Default|GetHF_RestoreHUDAfterSequenceDelaySeconds))))
          (if (<= _output_get 0)
            (return false)
            (else
              (LocalVariables|SetCurrentTime (Utilities|Time|GetTimeSeconds (Variables|Default|GetVehicleMod)))
              (Utilities|FlowControl|Sequence
                (:then_0
                  ; Record return-from-hidden time
                  (if (< (Variables|Default|GetReturnedFromHiddenTimestamp) 0)
                    (Variables|Default|SetReturnedFromHiddenTimestamp (LocalVariables|GetCurrentTime))))
                (:then_1
                  (return (< (- (LocalVariables|GetCurrentTime) (Variables|Default|GetReturnedFromHiddenTimestamp)) (LocalVariables|GetDelaySeconds)))))))
          (else
            (return (> (Variables|Default|GetIntroSequenceBlendedPercent) (Variables|Default|GetHideHudIntroSequencePercentage)))))))
    (:else)))

(fn ShouldShowTimeTunnel
  (bind _observingplayercontroller (Variables|Default|GetObservingPlayerController)
    (:then
      (bind _playercameramanager (Variables|/Script/Engine.PlayerController|GetPlayerCameraManager _observingplayercontroller)
        (:then
          (bind _vehiclemod (Variables|Default|GetVehicleMod)
            (:then
              (if (Class|VMBGAGreaseClipDunmurry|GetIsVehicleHiddenLocal _vehiclemod)
                (return true _playercameramanager)
                (else
                  _)))
            (:else
              _))
          (return false 0))
        (:else
          _)))
    (:else
      _)))

(fn UpdateIntroSequenceProgress
  (params (DeltaTime :type Float%20%28double-precision%29))
  (Utilities|FlowControl|Sequence
    (:then_0
      (bind _vehiclemod (Variables|Default|GetVehicleMod)
        (:then
          (bind (_introsequenceactive _activeprecent) (Class|VMBGAGreaseClipDunmurry|GetIntroSequencePercentage _vehiclemod))
          (Variables|Default|SetIntroSequenceActive _introsequenceactive)
          (LocalVariables|SetTargetBlendedPercent _activeprecent))
        (:else
          (Variables|Default|SetIntroSequenceActive false)
          (LocalVariables|SetTargetBlendedPercent 0.0))))
    (:then_1
      (Variables|Default|SetIntroSequenceBlendedPercent (Math|Float|Clamp(Float) (Math|Float|Clamp(Float) (LocalVariables|GetTargetBlendedPercent) (- (Variables|Default|GetIntroSequenceBlendedPercent) (/ DeltaTime (Variables|Default|GetIntroSequenceBlendOutSeconds))) (+ (Variables|Default|GetIntroSequenceBlendedPercent) (/ DeltaTime (Variables|Default|GetIntroSequenceCatchUpSeconds)))))))))

(fn ConstructionScript)
