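; Looping gameplay cue for the GreaseClipDunmurry vehicle mod, named for its vehicle intro sequence.
; When looping starts, it casts the cue's instigator to VM_BGA_GreaseClipDunmurry and stores it as the local vehicle mod.
; It then queries that mod for its attached vehicle and caches the result on the successful path.
; The removal event and construction script are empty, and the graph itself triggers no effects.
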(event EventOnLoopingStart(EitherVFX-preferthisversion)
  (params (MyTarget :type Actor%20Object%20Reference) (Parameters :type Gameplay%20Cue%20Parameters%20Structure%20%28by%20ref%29) (ParticleComponents :type Array%20of%20FXSystem%20Component%20Object%20References) (AudioComponents :type Array%20of%20Audio%20Component%20Object%20References))
  ;[1 1 1] Start / End Effect
  (bind _asvm_bga_grease_clip_dunmurry (Utilities|Casting|CastToVM_BGA_GreaseClipDunmurry (Ability|GameplayCue|GetInstigatorActor Parameters)))
  (bind _output_get (Variables|Default|SetLocal_VehicleMod _asvm_bga_grease_clip_dunmurry))
  (bind _attachedvehicle (Variables|/GreaseClipDunmurry/Gameplay/VM_BGA_GreaseClipDunmurry.VM_BGA_GreaseClipDunmurry_C|GetAttachedVehicle _output_get)
    (:then
      (Variables|Default|SetAttached%20Vehicle _attachedvehicle))
    (:else)))

(event EventOnRemoval(EitherVFX-preferthisversion)
  (params (MyTarget :type Actor%20Object%20Reference) (Parameters :type Gameplay%20Cue%20Parameters%20Structure%20%28by%20ref%29) (ParticleComponents :type Array%20of%20FXSystem%20Component%20Object%20References) (AudioComponents :type Array%20of%20Audio%20Component%20Object%20References) (BurstCameraShakeInstance :type Camera%20Shake%20Base%20Object%20Reference) (BurstDecalInstance :type Decal%20Actor%20Object%20Reference)))

(fn ConstructionScript)
