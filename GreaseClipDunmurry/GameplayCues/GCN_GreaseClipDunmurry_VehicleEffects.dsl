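; Burst gameplay cue that plays visual effects at the vehicle a GreaseClipDunmurry vehicle mod is attached to.
; On burst, it casts the cue's instigator to VM_BGA_GreaseClipDunmurry and gets that mod's attached vehicle.
; If that lookup succeeds, it spawns fire trail and reverse explosion Niagara systems at the vehicle's location and rotation, unattached.
; The overlap, tick and construction script entries are empty.
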(event Collision|EventActorBeginOverlap
  (params (OtherActor :type Actor%20Object%20Reference)))

(event EventTick
  (params (DeltaSeconds :type Float%20%28single-precision%29)))

(fn OnBurst(EitherVFX-preferthisversion)
  (params (MyTarget :type Actor%20Object%20Reference) (Parameters :type Gameplay%20Cue%20Parameters%20Structure%20%28by%20ref%29) (ParticleComponents :type Array%20of%20FXSystem%20Component%20Object%20References) (AudioComponents :type Array%20of%20Audio%20Component%20Object%20References) (BurstCameraShakeInstance :type Camera%20Shake%20Base%20Object%20Reference) (BurstDecalInstance :type Decal%20Actor%20Object%20Reference))
  (bind (_normalizedmagnitude _rawmagnitude _effectcontext _matchedtagname _originaltag _aggregatedsourcetags _aggregatedtargettags _location _normal _instigator) (Ability|GameplayCue|BreakGameplayCueParameters Parameters))
  (bind _asvm_bga_grease_clip_dunmurry (Utilities|Casting|CastToVM_BGA_GreaseClipDunmurry _instigator))
  (bind _output_get (LocalVariables|SetLocal_VehicleMod _asvm_bga_grease_clip_dunmurry))
  (bind _attachedvehicle (Variables|/GreaseClipDunmurry/Gameplay/VM_BGA_GreaseClipDunmurry.VM_BGA_GreaseClipDunmurry_C|GetAttachedVehicle _output_get)
    (:then
      (Niagara|SpawnSystematLocation _attachedvehicle "/GreaseClipDunmurry/Assets/FX/NS_GreaseClipDunmurry_FireTrails.NS_GreaseClipDunmurry_FireTrails" (Transformation|GetActorLocation _attachedvehicle) (Transformation|GetActorRotation _attachedvehicle) "1.000000,1.000000,1.000000" false)
      (Niagara|SpawnSystematLocation 0 "/GreaseClipDunmurry/Assets/FX/NS_GreaseClipDunmurry_ReverseExplosion.NS_GreaseClipDunmurry_ReverseExplosion" (Transformation|GetActorLocation _attachedvehicle) (Transformation|GetActorRotation _attachedvehicle) "1.000000,1.000000,1.000000" false))
    (:else)))

(fn ConstructionScript)
